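# Sound cue for the slam of a Floatmax interior double door.
# A single slam wave player feeds a Modulator with slight pitch variation, which connects to all four inputs of the root quality level node, so every level sounds the same.
# It uses the Fort_Doors sound class and attenuation settings, lowers the volume, and overrides concurrency to prevent new instances once the limit is reached.
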SoundCue /Game/Sounds/Helios/Environment_Actors/Door_Floatmax_Double/Door_Floatmax_Int_DoubleDoor_A_A_01_Slam_Cue
root=QualityLevel_0
nodes [
 {id=Modulator_0, children [WavePlayer_2], properties {pitchMax=0.99, pitchMin=1.01}}
 {id=QualityLevel_0, children [Modulator_0,Modulator_0,Modulator_0,Modulator_0]}
 {
  id=WavePlayer_2
  properties {soundWaveAssetPtr=&/Door_Floatmax_Int_DoubleDoor_A_A_01_Slam}
 }
]
properties {
 attenuationSettings=/Game/Sounds/Attenuation/Fort_Doors_Att
 bOverrideConcurrency=true
 maxDistance=2400
 soundClassObject=&Fort_Doors
 volumeMultiplier=0.6
 attenuationOverrides {falloffDistance=1600}
 concurrencyOverrides {resolutionRule=PreventNew}
}
